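# Animation Blueprint for the Bacon head part of the large male ZebraScramble NPC character, targeting that head's own skeleton.
# On every animation update it calls the parent update, then overrides the part skeletal mesh component's minimum LOD to 1, so LOD 0 is excluded.
# It declares no variables, and its pawn owner lookup is left unconnected.
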AnimBlueprint /BRCosmetics/Characters/Player/Male/Large/Bodies/M_LRG_ZebraScramble_Eggs_NPC/M_LRG_ZebraScramble_Bacon_Head_NPC_AnimBP
target_skeleton=/Game/Characters/Player/Male/Large/Heads/M_LRG_ZebraScramble_Bacon_Head/Meshes/M_LRG_ZebraScramble_Bacon_Head_Skeleton
graphs [EventGraph {nodes [
   n1 = Event BlueprintUpdateAnimation()
   > Parent: BlueprintUpdateAnimation!(execute, DeltaTimeX:float=@n1.DeltaTimeX) -> (then) {is=title K2Node_CallParentFunction}
   > OverrideMinLOD!(execute, self:Skinned Mesh Component Object Reference=GetPartSkeletalMeshComponent(), InNewMinLOD:int=1) -> (then)
   TryGetPawnOwner()
  ]}]
variables []
