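# Control Rig for the ZebraScramble Eggs large male NPC body that adjusts the head bone's rotation, limiting its range on one axis.
# It reads the head's local rotation as Euler angles and remaps Z through a cubic rich curve into a narrower range.
# X and Y pass through unchanged before the quaternion is rebuilt and written back to the head at full weight.
# An unconnected branch node is left in the graph.
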ControlRigBlueprint /BRCosmetics/Characters/Player/Male/Large/Bodies/M_LRG_ZebraScramble_Eggs_NPC/M_LRG_ZebraScramble_Eggs_NPC_CtrlRig
graphs [
 RigVMFunctionLibrary {links [], nodes []}
 RigVMModel {nodes [
   RigUnit_BeginExecution = &RigUnit_BeginExecution()
   RigVMFunction_MathQuaternionToEuler = &RigVMFunction_MathQuaternionToEuler(
    RotationOrder=ZYX
    Value=&RigUnit_GetTransform(Item=(Type=Bone,Name="head"), Space=LocalSpace, bInitial=False).Transform.Rotation
   )
   RigVMFunction_AnimEvalRichCurve = &RigVMFunction_AnimEvalRichCurve(
    Curve=(EditorCurveData=(Keys=((),(InterpMode=RCIM_Cubic,Time=0.6897,Value=0.8554,ArriveTangent=1,LeaveTangent=1),(Time=1,Value=1))))
    SourceMaximum=-45
    SourceMinimum=10
    TargetMaximum=-20
    TargetMinimum=10
    Value=@RigVMFunction_MathQuaternionToEuler.Result.Z
   )
   RigVMFunction_MathQuaternionFromEuler = &RigVMFunction_MathQuaternionFromEuler(
    Euler=(X=0,Y=0,Z=0)
    Euler.Z=@RigVMFunction_AnimEvalRichCurve.Result
    Euler.X=@RigVMFunction_MathQuaternionToEuler.Result.X
    Euler.Y=@RigVMFunction_MathQuaternionToEuler.Result.Y
   )
   RigUnit_SetRotation = &RigUnit_SetRotation(
    Item=(Type=Bone,Name="head")
    Weight=1
    bInitial=False
    ExecutePin=@RigUnit_BeginExecution.ExecutePin
    Value=@RigVMFunction_MathQuaternionFromEuler.Result
   )
   RigVMFunction_ControlFlowBranch = &RigVMFunction_ControlFlowBranch(Condition=False, ExecuteContext=())
   RigVMRerouteNode(Value=@RigVMFunction_ControlFlowBranch.False)
  ]}
]
hierarchy [
 root {loc=<0 0 0>, rot=<0 0 0 1>}
 pelvis {parent=root, loc=<0 -0.7321 92.4218>, rot=<0.0127 0.707 -0.0127 -0.707>}
 spine_01 {parent=pelvis, loc=<5.6002 -0.5128 0>, rot=<0 0 0.034 -0.9994>}
 spine_02 {parent=spine_01, loc=<8.0753 0.0139 0>, rot=<0 0 0.0238 -0.9997>}
 spine_03 {parent=spine_02, loc=<7.0393 0 0>, rot=<0 0 0.0149 -0.9999>}
 spine_04 {parent=spine_03, loc=<8.1024 0 0>, rot=<0 0 -0.0132 -0.9999>}
 spine_05 {parent=spine_04, loc=<10.0146 0 0>, rot=<0 0 -0.0015 -1>}
 attach_cape {parent=spine_05, loc=<9.1509 -13.3783 0>, rot=<0.0616 -0.7044 -0.0616 -0.7044>}
 attach_backpack {parent=spine_05, loc=<-0.123 -15.1423 0>, rot=<-0.0093 -0.707 0.0093 -0.707>}
 neck_01 {parent=spine_05, loc=<10.815 1.9196 0>, rot=<0 0 -0.1864 -0.9825>}
 neck_02 {parent=neck_01, loc=<4.7347 0.3045 0>, rot=<0 0 0.0363 -0.9993>}
 head {parent=neck_02, loc=<5.7632 0.5552 0>, rot=<0 0 0.115 -0.9934>}
 brow_glasses_l_root {parent=head, loc=<10.2608 7.8275 -4.6862>, rot=<0 0 0 -1>}
 brow_glasses_l_03 {parent=brow_glasses_l_root, loc=<3.3919 0.1997 -0.0171>, rot=<0 0 0 1>}
 brow_glasses_l_02 {parent=brow_glasses_l_root, loc=<2.818 0.341 1.862>, rot=<0 0.2895 0 -0.9572>}
 brow_glasses_l_04 {parent=brow_glasses_l_root, loc=<2.8219 0.0495 -1.8972>, rot=<0 -0.2737 0 -0.9618>}
 brow_glasses_l_01 {parent=brow_glasses_l_root, loc=<1.295 0.425 3.107>, rot=<0 0.5877 0 -0.8091>}
 brow_glasses_l_05 {parent=brow_glasses_l_root, loc=<1.588 0.0517 -2.9874>, rot=<0 -0.5359 0 -0.8443>}
 brow_glasses_r_root {parent=head, loc=<10.2605 7.8275 4.6862>, rot=<0 0 1 0>}
 brow_glasses_r_03 {parent=brow_glasses_r_root, loc=<-3.3925 -0.1997 0.0171>, rot=<0 0 0 1>}
 brow_glasses_r_02 {parent=brow_glasses_r_root, loc=<-2.8183 -0.3406 -1.8619>, rot=<0 0.2895 0 -0.9572>}
 brow_glasses_r_04 {parent=brow_glasses_r_root, loc=<-2.8227 -0.0495 1.8972>, rot=<0 -0.2737 0 -0.9618>}
 brow_glasses_r_01 {parent=brow_glasses_r_root, loc=<-1.2945 -0.4245 -3.1071>, rot=<0 0.5877 0 -0.8091>}
 brow_glasses_r_05 {parent=brow_glasses_r_root, loc=<-1.5886 -0.0517 2.9874>, rot=<0 -0.5359 0 -0.8443>}
 neck_twist {parent=neck_01, loc=<4.7347 0.3045 0>, rot=<0 0 0.0363 -0.9993>}
 neck_twist_lower {parent=spine_05, loc=<10.815 1.9196 0>, rot=<0 0 -0.1506 -0.9886>}
 clavicle_l {parent=spine_05, loc=<2.5856 -0.2866 -4.3777>, rot=<0.0465 -0.7289 -0.009 -0.6829>}
 deform_upperarm_l {parent=clavicle_l, loc=<18.2264 0.1342 -0.001>, rot=<0.0253 -0.2693 0.0013 -0.9627>}
 deform_lowerarm_l {parent=deform_upperarm_l, loc=<29.0071 0.1686 -1.239>, rot=<0.0391 -0.0099 -0.2453 -0.9686>}
 deform_hand_l {parent=deform_lowerarm_l, loc=<25.8607 0.222 1.0274>, rot=<-0.7193 -0.0988 0.1193 -0.6773>}
 deform_lowerarm_twist_02_l {parent=deform_lowerarm_l, loc=<9.6199 0 0>, rot=<0 0 0 -1>}
 deform_lowerarm_twist_01_l {parent=deform_lowerarm_l, loc=<17.8549 0 0>, rot=<0 0 0 -1>}
 deform_elbow_in_base_l {parent=deform_lowerarm_l, loc=<0.2247 5.7055 0.582>, rot=<-0.0013 -0.0043 -0.0011 -1>}
 deform_elbow_in_l {parent=deform_elbow_in_base_l, loc=<0 0 0>, rot=<0 0 0 1>}
 deform_elbow_out_base_l {parent=deform_lowerarm_l, loc=<0.0363 -7.4507 -0.2779>, rot=<-0.0013 -0.0043 -0.0011 -1>}
 deform_elbow_out_l {parent=deform_elbow_out_base_l, loc=<0 0 0>, rot=<0 0 0 1>}
 deform_elbowpad_base_l {parent=deform_lowerarm_l, loc=<0.804 -9.5161 -0.9861>, rot=<0 0 0 1>}
 deform_elbowpad_l {parent=deform_elbowpad_base_l, loc=<0 0 0>, rot=<0 0 0 1>}
 deform_upperarm_twist_01_l {parent=deform_upperarm_l, loc=<7.9796 0.0843 -0.6195>, rot=<0 0 0 -1>}
 deform_upperarm_bk_base_l {parent=deform_upperarm_twist_01_l, loc=<-6.8973 -13.3832 -1.3214>, rot=<0.0003 -0.0014 -0.0002 -1>}
 deform_upperarm_bk_l {parent=deform_upperarm_bk_base_l, loc=<0 0 0>, rot=<0 0 0 1>}
 deform_upperarm_fr_base_l {parent=deform_upperarm_twist_01_l, loc=<-8.3613 10.71 -2.1504>, rot=<0.0003 -0.0014 -0.0002 -1>}
 deform_upperarm_fr_l {parent=deform_upperarm_fr_base_l, loc=<0 0 0>, rot=<0 0 0 1>}
 deform_upperarm_up_base_l {parent=deform_upperarm_twist_01_l, loc=<-11.1634 -1.2183 8.7939>, rot=<0.0005 -0.0023 -0.0004 -1>}
 deform_upperarm_up_l {parent=deform_upperarm_up_base_l, loc=<0 0 0>, rot=<0 0 0 1>}
 deform_upperarm_armpit_base_l {parent=deform_upperarm_twist_01_l, loc=<3.3461 0.2222 -11.5685>, rot=<0.0056 -0.9668 -0.0406 -0.2524>}
 deform_upperarm_armpit_l {parent=deform_upperarm_armpit_base_l, loc=<0 0 0>, rot=<0 0 0 1>}
 deform_upperarm_twist_02_l {parent=deform_upperarm_l, loc=<19.5059 0.0843 -0.6195>, rot=<0 0 0 -1>}
 deform_bicep_base_l {parent=deform_upperarm_twist_02_l, loc=<-0.0192 8.0786 -0.1914>, rot=<0.0009 -0.0045 -0.0008 -1>}
 deform_bicep_l {parent=deform_bicep_base_l, loc=<0 0 0>, rot=<0 0 0 1>}
 upperarm_l {parent=clavicle_l, loc=<18.2264 0.1342 -0.001>, rot=<0.0253 -0.2693 0.0013 -0.9627>}
 lowerarm_l {parent=upperarm_l, loc=<29.0071 0.1686 -1.239>, rot=<0.0391 -0.0099 -0.2453 -0.9686>}
 hand_l {parent=lowerarm_l, loc=<25.8607 0.222 1.0274>, rot=<-0.7193 -0.0988 0.1193 -0.6773>}
 weapon_l {parent=hand_l, loc=<-0.1974 -3.4111 0.3811>, rot=<0 0 0 1>}
 thumb_01_l {parent=hand_l, loc=<1.3118 -0.3191 -2.459>, rot=<0.6546 -0.4042 0.0664 -0.6354>}
 thumb_02_l {parent=thumb_01_l, loc=<5.9336 -0.0448 0.0401>, rot=<-0.0037 0.0213 0.298 -0.9543>}
 thumb_03_l {parent=thumb_02_l, loc=<2.2223 0.3682 -0.0907>, rot=<0.0118 -0.013 0.0066 -0.9998>}
 index_metacarpal_l {parent=hand_l, loc=<1.6535 0.0183 -2.96>, rot=<0.0158 -0.0124 -0.0173 -0.9996>}
 index_01_l {parent=index_metacarpal_l, loc=<8.3008 0.0357 -0.0524>, rot=<-0.0044 0.0485 0.0924 -0.9945>}
 index_02_l {parent=index_01_l, loc=<3.3587 -0.3004 0.0452>, rot=<-0.002 0.0027 0.3225 -0.9466>}
 index_03_l {parent=index_02_l, loc=<2.5578 -0.123 -0.0252>, rot=<-0.002 0.0067 0.2558 -0.9667>}
 middle_metacarpal_l {parent=hand_l, loc=<2.1856 -0.1725 -0.5984>, rot=<0.006 0.0168 -0.0329 -0.9993>}
 middle_01_l {parent=middle_metacarpal_l, loc=<7.3275 -0.2447 0.0129>, rot=<-0.0208 0.0042 0.0858 -0.9961>}
 middle_02_l {parent=middle_01_l, loc=<4.0123 0.1401 0.1175>, rot=<-0.0218 0.0258 0.3512 -0.9357>}
 middle_03_l {parent=middle_02_l, loc=<2.6077 0.0534 0.1373>, rot=<0.0082 0.0188 0.1708 -0.9851>}
 ring_metacarpal_l {parent=hand_l, loc=<2.2074 -0.3625 1.3638>, rot=<-0.0284 0.0722 -0.0458 -0.9959>}
 ring_01_l {parent=ring_metacarpal_l, loc=<7.0295 -0.6197 -0.1446>, rot=<-0.0489 -0.0252 0.1892 -0.9804>}
 ring_02_l {parent=ring_01_l, loc=<3.6174 0.4647 -0.1027>, rot=<0 0 0.2806 -0.9598>}
 ring_03_l {parent=ring_02_l, loc=<2.4152 0.114 0.0199>, rot=<0 0 0.1189 -0.9929>}
 pinky_metacarpal_l {parent=hand_l, loc=<1.7616 -0.9753 3.3814>, rot=<-0.0397 0.0814 -0.069 -0.9935>}
 pinky_01_l {parent=pinky_metacarpal_l, loc=<7.1709 -0.4831 -0.1094>, rot=<-0.0654 -0.0035 0.2242 -0.9723>}
 pinky_02_l {parent=pinky_01_l, loc=<2.8274 0.0893 -0.1043>, rot=<0.0037 0.0011 0.2854 -0.9584>}
 pinky_03_l {parent=pinky_02_l, loc=<2.2046 0.0307 -0.0078>, rot=<-0.0002 0.0098 0.1445 -0.9895>}
 hand_attach_l {parent=hand_l, loc=<8.5 -3 0>, rot=<0.7071 0.7071 0 0>}
 deform_middle_metacarpal_l {parent=hand_l, loc=<2.3168 -0.634 0.8299>, rot=<0.0052 0.0156 -0.0284 -0.9995>}
 deform_middle_01_l {parent=deform_middle_metacarpal_l, loc=<7.3275 -0.2447 0.0129>, rot=<-0.0215 -0.0073 0.122 -0.9923>}
 deform_middle_02_l {parent=deform_middle_01_l, loc=<3.455 0.1401 0.1175>, rot=<-0.0308 -0.0039 0.3128 -0.9493>}
 deform_middle_03_l {parent=deform_middle_02_l, loc=<2.5266 0.0534 0.1373>, rot=<0.0082 0.0188 0.1708 -0.9851>}
 deform_wrist_dn_base_l {parent=hand_l, loc=<0.14 -3.8753 -0.0989>, rot=<-0.0008 -0.0012 0.0045 -1>}
 deform_wrist_dn_l {parent=deform_wrist_dn_base_l, loc=<0 0 0>, rot=<0 0 0 1>}
 deform_wrist_up_base_l {parent=hand_l, loc=<0.1951 2.3222 -0.0891>, rot=<-0.0008 -0.0012 0.0045 -1>}
 deform_wrist_up_l {parent=deform_wrist_up_base_l, loc=<0 0 0>, rot=<0 0 0 1>}
 lowerarm_twist_02_l {parent=lowerarm_l, loc=<9.6199 0 0>, rot=<0 0 0 -1>}
 lowerarm_twist_01_l {parent=lowerarm_l, loc=<17.8549 0 0>, rot=<0 0 0 -1>}
 upperarm_twist_01_l {parent=upperarm_l, loc=<7.9796 0.0843 -0.6195>, rot=<0 0 0 -1>}
 upperarm_twist_02_l {parent=upperarm_l, loc=<19.5059 0.0843 -0.6195>, rot=<0 0 0 -1>}
 deform_clavicle_l {parent=clavicle_l, loc=<0 0 0>, rot=<0 0 0 1>}
 clavicle_r {parent=spine_05, loc=<2.5855 -0.2866 4.3777>, rot=<0.7289 0.0465 -0.6829 0.009>}
 deform_upperarm_r {parent=clavicle_r, loc=<-18.2264 -0.1342 0.001>, rot=<0.0253 -0.2693 0.0013 -0.9627>}
 deform_lowerarm_r {parent=deform_upperarm_r, loc=<-29.0071 -0.1686 1.239>, rot=<0.0391 -0.0099 -0.2453 -0.9686>}
 deform_hand_r {parent=deform_lowerarm_r, loc=<-25.8607 -0.222 -1.0274>, rot=<-0.7193 -0.0988 0.1193 -0.6773>}
 deform_lowerarm_twist_02_r {parent=deform_lowerarm_r, loc=<-9.6199 0 0>, rot=<0 0 0 -1>}
 deform_lowerarm_twist_01_r {parent=deform_lowerarm_r, loc=<-17.8549 0 0>, rot=<0 0 0 -1>}
 deform_elbow_in_base_r {parent=deform_lowerarm_r, loc=<-0.2114 -5.6449 -0.827>, rot=<0 0 0 -1>}
 deform_elbow_in_r {parent=deform_elbow_in_base_r, loc=<0 0 0>, rot=<0 0 0 1>}
 deform_elbow_out_base_r {parent=deform_lowerarm_r, loc=<-0.0001 7.5129 0.0002>, rot=<0 0 0 -1>}
 deform_elbow_out_r {parent=deform_elbow_out_base_r, loc=<0 0 0>, rot=<0 0 0 1>}
 deform_elbowpad_base_r {parent=deform_lowerarm_r, loc=<-0.8042 9.5161 0.9863>, rot=<0 0 0 -1>}
 deform_elbowpad_r {parent=deform_elbowpad_base_r, loc=<0 0 0>, rot=<0 0 0 1>}
 deform_upperarm_twist_01_r {parent=deform_upperarm_r, loc=<-7.9796 -0.0843 0.6195>, rot=<0 0 0 -1>}
 deform_upperarm_bk_base_r {parent=deform_upperarm_twist_01_r, loc=<6.8985 13.3826 1.3257>, rot=<0 0 0 -1>}
 deform_upperarm_bk_r {parent=deform_upperarm_bk_base_r, loc=<0 0 0>, rot=<0 0 0 1>}
 deform_upperarm_fr_base_r {parent=deform_upperarm_twist_01_r, loc=<8.3485 -10.7117 2.1456>, rot=<0 0 0 -1>}
 deform_upperarm_fr_r {parent=deform_upperarm_fr_base_r, loc=<0 0 0>, rot=<0 0 0 1>}
 deform_upperarm_up_base_r {parent=deform_upperarm_twist_01_r, loc=<11.2015 1.2234 -8.7788>, rot=<0 0 0 -1>}
 deform_upperarm_up_r {parent=deform_upperarm_up_base_r, loc=<0 0 0>, rot=<0 0 0 1>}
 deform_upperarm_armpit_base_r {parent=deform_upperarm_twist_01_r, loc=<-3.3462 -0.2222 11.5687>, rot=<0.0056 -0.9668 -0.0406 -0.2524>}
 deform_upperarm_armpit_r {parent=deform_upperarm_armpit_base_r, loc=<0 0 0>, rot=<0 0 0 1>}
 deform_upperarm_twist_02_r {parent=deform_upperarm_r, loc=<-19.5059 -0.0843 0.6195>, rot=<0 0 0 -1>}
 deform_bicep_base_r {parent=deform_upperarm_twist_02_r, loc=<0.0002 -8.0494 -0.0003>, rot=<0 0 0 1>}
 deform_bicep_r {parent=deform_bicep_base_r, loc=<0 0 0>, rot=<0 0 0 1>}
 upperarm_r {parent=clavicle_r, loc=<-18.2264 -0.1342 0.001>, rot=<0.0253 -0.2693 0.0013 -0.9627>}
 lowerarm_r {parent=upperarm_r, loc=<-29.0071 -0.1686 1.239>, rot=<0.0391 -0.0099 -0.2453 -0.9686>}
 hand_r {parent=lowerarm_r, loc=<-25.8607 -0.222 -1.0274>, rot=<-0.7193 -0.0988 0.1193 -0.6773>}
 weapon_r {parent=hand_r, loc=<0.1974 3.4111 -0.3811>, rot=<0 0 0 1>}
 thumb_01_r {parent=hand_r, loc=<-1.3118 0.3191 2.459>, rot=<0.6546 -0.4042 0.0664 -0.6354>}
 thumb_02_r {parent=thumb_01_r, loc=<-5.9336 0.0448 -0.0401>, rot=<-0.0037 0.0213 0.298 -0.9543>}
 thumb_03_r {parent=thumb_02_r, loc=<-2.2223 -0.3682 0.0907>, rot=<0.0118 -0.013 0.0066 -0.9998>}
 index_metacarpal_r {parent=hand_r, loc=<-1.6535 -0.0183 2.96>, rot=<0.0158 -0.0124 -0.0173 -0.9996>}
 index_01_r {parent=index_metacarpal_r, loc=<-8.3008 -0.0357 0.0524>, rot=<-0.0044 0.0485 0.0924 -0.9945>}
 index_02_r {parent=index_01_r, loc=<-3.3587 0.3004 -0.0452>, rot=<-0.002 0.0027 0.3225 -0.9466>}
 index_03_r {parent=index_02_r, loc=<-2.5578 0.123 0.0252>, rot=<-0.002 0.0067 0.2558 -0.9667>}
 middle_metacarpal_r {parent=hand_r, loc=<-2.1856 0.1725 0.5984>, rot=<0.006 0.0168 -0.0329 -0.9993>}
 middle_01_r {parent=middle_metacarpal_r, loc=<-7.3275 0.2447 -0.0129>, rot=<-0.0208 0.0042 0.0858 -0.9961>}
 middle_02_r {parent=middle_01_r, loc=<-4.0123 -0.1401 -0.1175>, rot=<-0.0218 0.0258 0.3512 -0.9357>}
 middle_03_r {parent=middle_02_r, loc=<-2.6077 -0.0534 -0.1373>, rot=<0.0082 0.0188 0.1708 -0.9851>}
 ring_metacarpal_r {parent=hand_r, loc=<-2.2074 0.3625 -1.3638>, rot=<-0.0284 0.0722 -0.0458 -0.9959>}
 ring_01_r {parent=ring_metacarpal_r, loc=<-7.0295 0.6197 0.1446>, rot=<-0.0489 -0.0252 0.1892 -0.9804>}
 ring_02_r {parent=ring_01_r, loc=<-3.6174 -0.4647 0.1027>, rot=<0 0 0.2806 -0.9598>}
 ring_03_r {parent=ring_02_r, loc=<-2.4152 -0.114 -0.0199>, rot=<0 0 0.1189 -0.9929>}
 pinky_metacarpal_r {parent=hand_r, loc=<-1.7616 0.9753 -3.3814>, rot=<-0.0397 0.0814 -0.069 -0.9935>}
 pinky_01_r {parent=pinky_metacarpal_r, loc=<-7.1709 0.4831 0.1094>, rot=<-0.0654 -0.0035 0.2242 -0.9723>}
 pinky_02_r {parent=pinky_01_r, loc=<-2.8274 -0.0893 0.1043>, rot=<0.0037 0.0011 0.2854 -0.9584>}
 pinky_03_r {parent=pinky_02_r, loc=<-2.2046 -0.0307 0.0078>, rot=<-0.0002 0.0098 0.1445 -0.9895>}
 hand_attach_r {parent=hand_r, loc=<-8.5 3 0>, rot=<0 0 -0.7071 -0.7071>}
 deform_middle_metacarpal_r {parent=hand_r, loc=<-2.1455 0.1552 -0.9288>, rot=<0.006 0.0168 -0.0329 -0.9993>}
 deform_middle_01_r {parent=deform_middle_metacarpal_r, loc=<-7.3275 0.2448 -0.0129>, rot=<-0.0215 -0.0073 0.122 -0.9923>}
 deform_middle_02_r {parent=deform_middle_01_r, loc=<-3.4551 -0.1401 -0.1175>, rot=<-0.0308 -0.0039 0.3128 -0.9493>}
 deform_middle_03_r {parent=deform_middle_02_r, loc=<-2.5267 -0.0534 -0.1373>, rot=<0.0082 0.0188 0.1708 -0.9851>}
 deform_wrist_dn_base_r {parent=hand_r, loc=<0 3.4172 0>, rot=<0 0 0 -1>}
 deform_wrist_dn_r {parent=deform_wrist_dn_base_r, loc=<0 0 0>, rot=<0 0 0 -1>}
 deform_wrist_up_base_r {parent=hand_r, loc=<-0.0001 -2.7804 0>, rot=<0 0 0 -1>}
 deform_wrist_up_r {parent=deform_wrist_up_base_r, loc=<0 0 0>, rot=<0 0 0 -1>}
 lowerarm_twist_02_r {parent=lowerarm_r, loc=<-9.6199 0 0>, rot=<0 0 0 -1>}
 lowerarm_twist_01_r {parent=lowerarm_r, loc=<-17.8549 0 0>, rot=<0 0 0 -1>}
 upperarm_twist_01_r {parent=upperarm_r, loc=<-7.9796 -0.0843 0.6195>, rot=<0 0 0 -1>}
 upperarm_twist_02_r {parent=upperarm_r, loc=<-19.5059 -0.0843 0.6195>, rot=<0 0 0 -1>}
 deform_clavicle_r {parent=clavicle_r, loc=<0 0 0>, rot=<0 0 0 1>}
 deform_pec_aim_base_l {parent=spine_04, loc=<5.0232 13.6719 -2.6508>, rot=<0.0672 -0.5201 0.3228 -0.7879>}
 deform_pec_aim_l {parent=deform_pec_aim_base_l, loc=<0 0 0>, rot=<0 0 0 1>}
 deform_pec_base_l {parent=deform_pec_aim_l, loc=<16.3279 0 0>, rot=<0 0 0 1>}
 deform_pec_l {parent=deform_pec_base_l, loc=<0 0 0>, rot=<0 0 0 1>}
 deform_pec_aim_base_r {parent=spine_04, loc=<5.023 13.6718 2.6508>, rot=<-0.5201 -0.0672 0.7879 0.3228>}
 deform_pec_aim_r {parent=deform_pec_aim_base_r, loc=<0 0 0>, rot=<0 0 0 1>}
 deform_pec_base_r {parent=deform_pec_aim_r, loc=<-16.3279 0 0>, rot=<0 0 0 -1>}
 deform_pec_r {parent=deform_pec_base_r, loc=<0 0 0.0002>, rot=<0 0 0 -1>}
 deform_neckline_l_01_root {parent=spine_04, loc=<12.5886 18.3289 -11.6054>, rot=<0 0 0 1>}
 deform_neckline_l_01 {parent=deform_neckline_l_01_root, loc=<0 0 0>, rot=<0 0 0 1>}
 deform_neckline_l_02_root {parent=spine_04, loc=<22.046 1.7134 -15.9036>, rot=<0 0 0 1>}
 deform_neckline_l_02 {parent=deform_neckline_l_02_root, loc=<0 0 0>, rot=<0 0 0 1>}
 deform_neckline_l_03_root {parent=spine_04, loc=<26.0578 -8.0793 -10.6447>, rot=<0 0 0 1>}
 deform_neckline_l_03 {parent=deform_neckline_l_03_root, loc=<0 0 0>, rot=<0 0 0 1>}
 deform_neckline_rear_root {parent=spine_04, loc=<27.3206 -11.3828 0>, rot=<0 0 0 1>}
 deform_neckline_rear {parent=deform_neckline_rear_root, loc=<0 0 0>, rot=<0 0 0 1>}
 deform_neckline_r_01_root {parent=spine_04, loc=<12.5881 18.3289 11.6054>, rot=<0 0 1 0>}
 deform_neckline_r_01 {parent=deform_neckline_r_01_root, loc=<0 0 0>, rot=<0 0 1 0>}
 deform_neckline_r_02_root {parent=spine_04, loc=<22.0458 1.7134 15.9036>, rot=<0 0 1 0>}
 deform_neckline_r_02 {parent=deform_neckline_r_02_root, loc=<0 0 0>, rot=<0 0 1 0>}
 deform_neckline_r_03_root {parent=spine_04, loc=<26.0576 -8.0793 10.6447>, rot=<0 0 1 0>}
 deform_neckline_r_03 {parent=deform_neckline_r_03_root, loc=<0 0 0>, rot=<0 0 1 0>}
 deform_lat_l_base {parent=spine_03, loc=<2.5566 -10.3716 -17.6918>, rot=<0 0 0 1>}
 deform_lat_l {parent=deform_lat_l_base, loc=<0 0 0>, rot=<0 0 0 1>}
 deform_lat_r_base {parent=spine_03, loc=<2.5568 -10.3716 17.6918>, rot=<0 0 1 0>}
 deform_lat_r {parent=deform_lat_r_base, loc=<0 0 0>, rot=<0 0 0 -1>}
 thigh_l {parent=pelvis, loc=<-1.8716 -0.0109 -11.9936>, rot=<0.0012 0.0572 0.0155 -0.9982>}
 calf_l {parent=thigh_l, loc=<-41.9178 -0.2246 0.0926>, rot=<0.0006 -0.0161 -0.0391 -0.9991>}
 foot_l {parent=calf_l, loc=<-38.7057 1.2438 0.2452>, rot=<0.0975 -0.0474 0.0317 -0.9936>}
 ball_l {parent=foot_l, loc=<-6.1614 13.1469 0.6807>, rot=<0.0001 -0.0001 -0.7063 -0.7079>}
 deform_foot_out_base_l {parent=foot_l, loc=<2.3108 6.348 1.3105>, rot=<-0.0004 0.0107 0.0021 -0.9999>}
 deform_foot_out_l {parent=deform_foot_out_base_l, loc=<0 0 0>, rot=<0 0 0 1>}
 deform_foot_in_base_l {parent=foot_l, loc=<1.3397 -10.8214 -0.009>, rot=<-0.0004 0 -0.0001 -1>}
 deform_foot_in_l {parent=deform_foot_in_base_l, loc=<0 0 0>, rot=<0 0 0 1>}
 calf_twist_02_l {parent=calf_l, loc=<-15.9268 -0.0967 0.0397>, rot=<0.0237 0.0095 -0.0035 -0.9997>}
 calf_twist_01_l {parent=calf_l, loc=<-32.9335 -0.0967 0.0397>, rot=<0.0237 0.0095 -0.0035 -0.9997>}
 thigh_twist_01_l {parent=thigh_l, loc=<-21.6303 0 0>, rot=<0.0013 0 0 -1>}
 thigh_r {parent=pelvis, loc=<-1.8716 -0.0109 11.9936>, rot=<0.0572 -0.0012 0.9982 0.0155>}
 calf_r {parent=thigh_r, loc=<41.9178 0.2246 -0.0926>, rot=<0.0006 -0.0161 -0.0391 -0.9991>}
 foot_r {parent=calf_r, loc=<38.7057 -1.2438 -0.2452>, rot=<0.0975 -0.0474 0.0317 -0.9936>}
 ball_r {parent=foot_r, loc=<6.1614 -13.1469 -0.6807>, rot=<0.0001 -0.0001 -0.7063 -0.7079>}
 deform_foot_out_base_r {parent=foot_r, loc=<-2.3115 -6.3488 -1.3052>, rot=<0.0001 0.0107 0.0022 -0.9999>}
 deform_foot_out_r {parent=deform_foot_out_base_r, loc=<0 0 0>, rot=<0 0 0 -1>}
 deform_foot_in_base_r {parent=foot_r, loc=<-1.3384 10.8215 0>, rot=<0 0 0 -1>}
 deform_foot_in_r {parent=deform_foot_in_base_r, loc=<0 0 0>, rot=<0 0 0 -1>}
 calf_twist_02_r {parent=calf_r, loc=<15.9268 0.0967 -0.0397>, rot=<0.0237 0.0095 -0.0035 -0.9997>}
 calf_twist_01_r {parent=calf_r, loc=<32.9335 0.0967 -0.0397>, rot=<0.0237 0.0095 -0.0035 -0.9997>}
 thigh_twist_01_r {parent=thigh_r, loc=<21.6303 0 0>, rot=<0.0013 0 0 -1>}
 deform_thigh_r_root {parent=pelvis, loc=<-8.7389 0.922 10.057>, rot=<0.068 -0.0014 0.9976 0.0154>}
 deform_thigh_r {parent=deform_thigh_r_root, loc=<0 0 0>, rot=<0 0 0 1>}
 deform_calf_r {parent=deform_thigh_r, loc=<40.0215 0 0.0001>, rot=<0.0006 -0.0161 -0.0391 -0.9991>}
 deform_foot_r {parent=deform_calf_r, loc=<34.0404 0.0647 0.6116>, rot=<0.0975 -0.0474 0.0317 -0.9936>}
 deform_calf_twist_02_r {parent=deform_calf_r, loc=<15.9267 0.0967 -0.0397>, rot=<0.0237 0.0095 -0.0035 -0.9997>}
 deform_calf_twist_01_r {parent=deform_calf_r, loc=<32.9335 0.0967 -0.0397>, rot=<0.0237 0.0095 -0.0035 -0.9997>}
 deform_eggs_knee_in_base_r {parent=deform_calf_r, loc=<0.151 8.9593 1.6583>, rot=<0 0 0 -1>}
 deform_eggs_knee_in_r {parent=deform_eggs_knee_in_base_r, loc=<0 0 0>, rot=<0 0 0 -1>}
 deform_eggs_knee_out_base_r {parent=deform_calf_r, loc=<-1.2443 -8.5388 1.6093>, rot=<-0.0009 -0.0109 0 -0.9999>}
 deform_eggs_knee_out_r {parent=deform_eggs_knee_out_base_r, loc=<0 0 0>, rot=<0 0 0 -1>}
 deform_thigh_twist_01_r {parent=deform_thigh_r, loc=<21.6302 0 0>, rot=<0.0013 0 0 -1>}
 deform_thigh_l_root {parent=pelvis, loc=<-8.7389 0.922 -10.057>, rot=<0.0014 0.068 0.0154 -0.9976>}
 deform_thigh_l {parent=deform_thigh_l_root, loc=<0 0 0>, rot=<-0.0591 -0.0022 0.0004 -0.9982>}
 deform_calf_l {parent=deform_thigh_l, loc=<-40.0215 0 0>, rot=<0.0006 -0.0161 -0.0391 -0.9991>}
 deform_foot_l {parent=deform_calf_l, loc=<-34.0403 -0.0647 -0.6116>, rot=<0.1557 -0.048 0.0357 -0.986>}
 deform_calf_twist_02_l {parent=deform_calf_l, loc=<-15.9268 -0.0967 0.0397>, rot=<0.0237 0.0095 -0.0035 -0.9997>}
 deform_calf_twist_01_l {parent=deform_calf_l, loc=<-32.9335 -0.0967 0.0397>, rot=<0.0237 0.0095 -0.0035 -0.9997>}
 deform_eggs_knee_in_base_l {parent=deform_calf_l, loc=<-0.1509 -8.9593 -1.6584>, rot=<0 0 0 -1>}
 deform_eggs_knee_in_l {parent=deform_eggs_knee_in_base_l, loc=<0 0 0>, rot=<0 0 0 1>}
 deform_eggs_knee_out_base_l {parent=deform_calf_l, loc=<1.2443 8.5388 -1.6093>, rot=<-0.0009 -0.0109 0 -0.9999>}
 deform_eggs_knee_out_l {parent=deform_eggs_knee_out_base_l, loc=<0 0 0>, rot=<0 0 0 1>}
 deform_thigh_twist_01_l {parent=deform_thigh_l, loc=<-21.6303 0 0>, rot=<0.0013 0 0 -1>}
 attach {parent=root, loc=<0 0 0>, rot=<0 0 0 1>}
 ik_hand_root {parent=root, loc=<0 0 0>, rot=<0 0 0 1>}
 ik_hand_gun {parent=ik_hand_root, loc=<-64.462 4.8638 101.1961>, rot=<0.6084 0.1793 0.4613 -0.6205>}
 ik_hand_r {parent=ik_hand_gun, loc=<0 0 0>, rot=<0 0 0 1>}
 ik_hand_l {parent=ik_hand_gun, loc=<65.7708 101.9258 43.6659>, rot=<-0.5102 -0.7906 -0.3387 0>}
 ik_foot_root {parent=root, loc=<0 0 0>, rot=<0 0 0 1>}
 ik_foot_l {parent=ik_foot_root, loc=<19.6189 -3.2051 10.3339>, rot=<0.0753 0.7018 0.0677 -0.7051>}
 ik_foot_r {parent=ik_foot_root, loc=<-19.6189 -3.2051 10.3339>, rot=<0.7051 0.0677 -0.7018 0.0753>}
 attach_fp {parent=root, loc=<0 0 0>, rot=<0 0 0 1>}
 camera_root {parent=root, loc=<0 0 0>, rot=<0 0 0 1>}
 camera {parent=camera_root, loc=<0 0 0>, rot=<0 0 0 1>}
 attach_emote_01 {parent=root, loc=<0 0 0>, rot=<0 0 0 1>}
 attach_emote_02 {parent=root, loc=<0 0 0>, rot=<0 0 0 1>}
 attach_emote_03 {parent=root, loc=<0 0 0>, rot=<0 0 0 1>}
 warp_point_01 {parent=root, loc=<0 0 0>, rot=<0 0 0 1>}
 warp_point_02 {parent=root, loc=<0 0 0>, rot=<0 0 0 1>}
 warp_point_03 {parent=root, loc=<0 0 0>, rot=<0 0 0 1>}
 disable_zebrascramble_finger {type=Curve, loc=<0 0 0>, rot=<0 0 0 1>}
 Eye_open_pose {type=Curve, loc=<0 0 0>, rot=<0 0 0 1>}
 L_blink_pose {type=Curve, loc=<0 0 0>, rot=<0 0 0 1>}
 L_squint_inner_pose {type=Curve, loc=<0 0 0>, rot=<0 0 0 1>}
 L_upperLidDown_pose {type=Curve, loc=<0 0 0>, rot=<0 0 0 1>}
 look_down_pose {type=Curve, loc=<0 0 0>, rot=<0 0 0 1>}
 look_left_pose {type=Curve, loc=<0 0 0>, rot=<0 0 0 1>}
 look_right_pose {type=Curve, loc=<0 0 0>, rot=<0 0 0 1>}
 look_up_pose {type=Curve, loc=<0 0 0>, rot=<0 0 0 1>}
 R_blink_pose {type=Curve, loc=<0 0 0>, rot=<0 0 0 1>}
 R_squint_inner_pose {type=Curve, loc=<0 0 0>, rot=<0 0 0 1>}
 R_upperLidDown_pose {type=Curve, loc=<0 0 0>, rot=<0 0 0 1>}
]
