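# Animation Blueprint for the NPC ZebraScramble Eggs cosmetic body, animating its glasses' brows from head-part facial curves.
# Copies the attached parent's pose, runs two Control Rigs, then layers additive brow_glasses bone offsets weighted by pose variables.
# Brow-down curves snap on or off by threshold and can suppress glabella poses; a brow-combo chain is unconnected.
# After evaluation, glasses root bone transforms feed eye virtual-plane vectors into a dynamic material.
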AnimBlueprint /BRCosmetics/Characters/Player/Male/Large/Bodies/M_LRG_ZebraScramble_Eggs_NPC/M_LRG_ZebraScramble_Eggs_AnimBP_NPC
target_skeleton=&/M_LRG_ZebraScramble_Eggs_NPC_Skeleton
graphs [
 AnimGraph {nodes [
   Component To Local()
   "Transform (Modify) Bone"(Alpha=1, Translation="0,0,0", Rotation="0,0,0", Scale="1,1,1") {alphaBoolBlend {blendOption=Linear}, boneToModify=~}
   n1 = Get L_brow_down_and_squeeze_combo_pose() -> (L_brow_down_and_squeeze_combo_pose:double)
   n2 = "Transform (Modify) Bone - Bone: brow_glasses_l_01"(Alpha=@n1, Translation="-0.388,-0.000,0.046", Rotation="-7.318,-0.002,0.063", Scale="1,1,1") {
    rotationMode=BMM_Additive
    rotationSpace=BCS_ParentBoneSpace
    scaleSpace=BCS_ParentBoneSpace
    translationMode=BMM_Additive
    translationSpace=BCS_ParentBoneSpace
    alphaBoolBlend {blendOption=Linear}
    boneToModify {boneName=brow_glasses_l_01}
    rotation=<rot -7.318 -0.002 0.063>
    translation {x=-0.388, z=0.046}
   }
   n3 = "Transform (Modify) Bone - Bone: brow_glasses_l_02"(ComponentPose=@n2, Alpha=@n1, Translation="0.048,-0.000,-0.000", Rotation="13.909,-0.021,-0.119", Scale="1.188,1.000,1.000") {
    rotationMode=BMM_Additive
    rotationSpace=BCS_ParentBoneSpace
    scaleMode=BMM_Additive
    scaleSpace=BCS_ParentBoneSpace
    translationMode=BMM_Additive
    translationSpace=BCS_ParentBoneSpace
    alphaBoolBlend {blendOption=Linear}
    boneToModify {boneName=brow_glasses_l_02}
    rotation=<rot 13.909 -0.021 -0.119>
    scale {x=1.188}
    translation {x=0.048}
   }
   n4 = "Transform (Modify) Bone - Bone: brow_glasses_l_03"(ComponentPose=@n3, Alpha=@n1, Translation="-0.073,0.002,0.000", Rotation="-4.435,0.000,0.037", Scale="1,1,1") {
    rotationMode=BMM_Additive
    rotationSpace=BCS_ParentBoneSpace
    scaleMode=BMM_Additive
    scaleSpace=BCS_ParentBoneSpace
    translationMode=BMM_Additive
    translationSpace=BCS_ParentBoneSpace
    alphaBoolBlend {blendOption=Linear}
    boneToModify {boneName=brow_glasses_l_03}
    rotation {pitch=-4.435, roll=0.037}
    translation=<-0.073 0.002>
   }
   n5 = "Transform (Modify) Bone - Bone: brow_glasses_l_04"(ComponentPose=@n4, Alpha=@n1, Translation="0.258,0.000,0.000", Rotation="30.360,-0.076,-0.279", Scale="1,1,1") {
    rotationMode=BMM_Additive
    rotationSpace=BCS_ParentBoneSpace
    scaleMode=BMM_Additive
    scaleSpace=BCS_ParentBoneSpace
    translationMode=BMM_Additive
    translationSpace=BCS_ParentBoneSpace
    alphaBoolBlend {blendOption=Linear}
    boneToModify {boneName=brow_glasses_l_04}
    rotation=<rot 30.36 -0.076 -0.279>
    translation {x=0.258}
   }
   n6 = "Transform (Modify) Bone - Bone: brow_glasses_l_05"(ComponentPose=@n5, Alpha=@n1, Translation="0,0,0", Rotation="0,0,0", Scale="1,1,1") {
    rotationMode=BMM_Additive
    rotationSpace=BCS_ParentBoneSpace
    scaleSpace=BCS_ParentBoneSpace
    translationMode=BMM_Additive
    translationSpace=BCS_ParentBoneSpace
    alphaBoolBlend {blendOption=Linear}
    boneToModify {boneName=brow_glasses_l_05}
   }
   n7 = Get R_brow_down_and_squeeze_combo_pose() -> (R_brow_down_and_squeeze_combo_pose:double)
   n8 = "Transform (Modify) Bone - Bone: brow_glasses_r_01"(ComponentPose=@n6, Alpha=@n7, Translation="0.408,0.000,-0.046", Rotation="-11.534,-0.008,0.103", Scale="1,1,1") {
    rotationMode=BMM_Additive
    rotationSpace=BCS_ParentBoneSpace
    scaleSpace=BCS_ParentBoneSpace
    translationMode=BMM_Additive
    translationSpace=BCS_ParentBoneSpace
    alphaBoolBlend {blendOption=Linear}
    boneToModify {boneName=brow_glasses_r_01}
    rotation=<rot -11.534 -0.008 0.103>
    translation=<0.4084 0 -0.046>
   }
   n9 = "Transform (Modify) Bone - Bone: brow_glasses_r_02"(ComponentPose=@n8, Alpha=@n7, Translation="0.273,-0.000,-0.000", Rotation="9.851,-0.014,-0.085", Scale="1.188,1,1") {
    rotationMode=BMM_Additive
    rotationSpace=BCS_ParentBoneSpace
    scaleSpace=BCS_ParentBoneSpace
    translationMode=BMM_Additive
    translationSpace=BCS_ParentBoneSpace
    alphaBoolBlend {blendOption=Linear}
    boneToModify {boneName=brow_glasses_r_02}
    rotation=<rot 9.851 -0.014 -0.085>
    scale {x=1.188}
    translation {x=0.273}
   }
   n10 = "Transform (Modify) Bone - Bone: brow_glasses_r_03"(ComponentPose=@n9, Alpha=@n7, Translation="0.073,0.002,0", Rotation="-4.435,0,0.037", Scale="1,1,1") {
    rotationMode=BMM_Additive
    rotationSpace=BCS_ParentBoneSpace
    scaleMode=BMM_Additive
    scaleSpace=BCS_ParentBoneSpace
    translationMode=BMM_Additive
    translationSpace=BCS_ParentBoneSpace
    alphaBoolBlend {blendOption=Linear}
    boneToModify {boneName=brow_glasses_r_03}
    rotation {pitch=-4.435, roll=0.037}
    translation=<0.073 0.002>
   }
   n11 = "Transform (Modify) Bone - Bone: brow_glasses_r_04"(ComponentPose=@n10, Alpha=@n7, Translation="-0.258,0,0", Rotation="30.36,-0.076,-0.279", Scale="1,1,1") {
    rotationMode=BMM_Additive
    rotationSpace=BCS_ParentBoneSpace
    scaleSpace=BCS_ParentBoneSpace
    translationMode=BMM_Additive
    translationSpace=BCS_ParentBoneSpace
    alphaBoolBlend {blendOption=Linear}
    boneToModify {boneName=brow_glasses_r_04}
    rotation=<rot 30.36 -0.076 -0.279>
    translation {x=-0.258}
   }
   "Transform (Modify) Bone - Bone: brow_glasses_r_05"(ComponentPose=@n11, Alpha=@n7, Translation="0,0,0", Rotation="0,0,0", Scale="1,1,1") {
    rotationMode=BMM_Additive
    rotationSpace=BCS_ParentBoneSpace
    scaleSpace=BCS_ParentBoneSpace
    translationMode=BMM_Additive
    translationSpace=BCS_ParentBoneSpace
    alphaBoolBlend {blendOption=Linear}
    boneToModify {boneName=brow_glasses_r_05}
   }
   n12 = Control Rig!(
    Source=Copy Pose From Mesh() {bCopyCurves=true, bUseAttachedParent=true}
    Alpha
    Is Emoting:bool=Get bIsPlayingEmote()
    disable_fingers:double=Get disable_zebrascramble_fingers() -> (disable_zebrascramble_fingers:float)
   ) -> (Pose) {
    alphaBoolBlend {blendOption=Linear}
    controlRigAssetReference {blueprintRigClass=/BRCosmetics/Characters/Player/Male/Large/Bodies/M_LRG_ZebraScramble_Eggs/Meshes/M_LRG_ZebraScramble_Eggs_CtrlRig.M_LRG_ZebraScramble_Eggs_CtrlRig_C}
   }
   n13 = Control Rig!(Source=@n12, Alpha) -> (Pose) {
    alphaBoolBlend {blendOption=Linear}
    controlRigAssetReference {blueprintRigClass=&/M_LRG_ZebraScramble_Eggs_NPC_CtrlRig.M_LRG_ZebraScramble_Eggs_NPC_CtrlRig_C}
   }
   n14 = Get L_brow_down_pose() -> (L_brow_down_pose:float)
   n15 = "Transform (Modify) Bone - Bone: brow_glasses_l_01"(ComponentPose=Local To Component(LocalPose=@n13), Alpha=@n14, Translation="0,0,0", Rotation="0,0,0", Scale="1,1,1") {
    rotationMode=BMM_Additive
    rotationSpace=BCS_ParentBoneSpace
    scaleSpace=BCS_ParentBoneSpace
    translationMode=BMM_Additive
    translationSpace=BCS_ParentBoneSpace
    alphaBoolBlend {blendOption=Linear}
    boneToModify {boneName=brow_glasses_l_01}
   }
   n16 = "Transform (Modify) Bone - Bone: brow_glasses_l_02"(ComponentPose=@n15, Alpha=@n14, Translation="-1.013,-0.004,-0.658", Rotation="-15.539,0,0", Scale="1,1,1") {
    rotationMode=BMM_Additive
    rotationSpace=BCS_ParentBoneSpace
    scaleMode=BMM_Additive
    scaleSpace=BCS_ParentBoneSpace
    translationMode=BMM_Additive
    translationSpace=BCS_ParentBoneSpace
    alphaBoolBlend {blendOption=Linear}
    boneToModify {boneName=brow_glasses_l_02}
    rotation {pitch=-15.539}
    translation=<-1.0126 -0.004 -0.658>
   }
   n17 = "Transform (Modify) Bone - Bone: brow_glasses_l_03"(ComponentPose=@n16, Alpha=@n14, Translation="-1.582,0.000,0.000", Rotation="0,0,0", Scale="1,1,1") {
    rotationMode=BMM_Additive
    rotationSpace=BCS_ParentBoneSpace
    scaleMode=BMM_Additive
    scaleSpace=BCS_ParentBoneSpace
    translationMode=BMM_Additive
    translationSpace=BCS_ParentBoneSpace
    alphaBoolBlend {blendOption=Linear}
    boneToModify {boneName=brow_glasses_l_03}
    translation {x=-1.582}
   }
   n18 = "Transform (Modify) Bone - Bone: brow_glasses_l_04"(ComponentPose=@n17, Alpha=@n14, Translation="-1.083,-0.005,0.000", Rotation="11.164,-0.007,-0.092", Scale="1,1,1") {
    rotationMode=BMM_Additive
    rotationSpace=BCS_ParentBoneSpace
    scaleMode=BMM_Additive
    scaleSpace=BCS_ParentBoneSpace
    translationMode=BMM_Additive
    translationSpace=BCS_ParentBoneSpace
    alphaBoolBlend {blendOption=Linear}
    boneToModify {boneName=brow_glasses_l_04}
    rotation=<rot 11.164 -0.007 -0.092>
    translation=<-1.083 -0.005>
   }
   n19 = "Transform (Modify) Bone - Bone: brow_glasses_l_05"(ComponentPose=@n18, Alpha=@n14, Translation="0,0,0", Rotation="0,0,0", Scale="1,1,1") {
    rotationMode=BMM_Additive
    rotationSpace=BCS_ParentBoneSpace
    scaleSpace=BCS_ParentBoneSpace
    translationMode=BMM_Additive
    translationSpace=BCS_ParentBoneSpace
    alphaBoolBlend {blendOption=Linear}
    boneToModify {boneName=brow_glasses_l_05}
   }
   n20 = Get R_brow_down_pose() -> (R_brow_down_pose:float)
   n21 = "Transform (Modify) Bone - Bone: brow_glasses_r_01"(ComponentPose=@n19, Alpha=@n20, Translation="0,0,0", Rotation="0,0,0", Scale="1,1,1") {
    rotationMode=BMM_Additive
    rotationSpace=BCS_ParentBoneSpace
    scaleSpace=BCS_ParentBoneSpace
    translationMode=BMM_Additive
    translationSpace=BCS_ParentBoneSpace
    alphaBoolBlend {blendOption=Linear}
    boneToModify {boneName=brow_glasses_r_01}
   }
   n22 = "Transform (Modify) Bone - Bone: brow_glasses_r_02"(ComponentPose=@n21, Alpha=@n20, Translation="1.013,-0.004,-0.658", Rotation="-15.539,0,0", Scale="1,1,1") {
    rotationMode=BMM_Additive
    rotationSpace=BCS_ParentBoneSpace
    scaleSpace=BCS_ParentBoneSpace
    translationMode=BMM_Additive
    translationSpace=BCS_ParentBoneSpace
    alphaBoolBlend {blendOption=Linear}
    boneToModify {boneName=brow_glasses_r_02}
    rotation {pitch=-15.539}
    translation=<1.013 -0.004 -0.658>
   }
   n23 = "Transform (Modify) Bone - Bone: brow_glasses_r_03"(ComponentPose=@n22, Alpha=@n20, Translation="1.582,0,0", Rotation="0,0,0", Scale="1,1,1") {
    rotationMode=BMM_Additive
    rotationSpace=BCS_ParentBoneSpace
    scaleMode=BMM_Additive
    scaleSpace=BCS_ParentBoneSpace
    translationMode=BMM_Additive
    translationSpace=BCS_ParentBoneSpace
    alphaBoolBlend {blendOption=Linear}
    boneToModify {boneName=brow_glasses_r_03}
    translation {x=1.582}
   }
   n24 = "Transform (Modify) Bone - Bone: brow_glasses_r_04"(ComponentPose=@n23, Alpha=@n20, Translation="1.083,0.005,0", Rotation="11.164,-0.007,0.092", Scale="1,1,1") {
    rotationMode=BMM_Additive
    rotationSpace=BCS_ParentBoneSpace
    scaleSpace=BCS_ParentBoneSpace
    translationMode=BMM_Additive
    translationSpace=BCS_ParentBoneSpace
    alphaBoolBlend {blendOption=Linear}
    boneToModify {boneName=brow_glasses_r_04}
    rotation=<rot 11.164 -0.007 0.092>
    translation=<1.083 0.005>
   }
   n25 = "Transform (Modify) Bone - Bone: brow_glasses_r_05"(ComponentPose=@n24, Alpha=@n20, Translation="0,0,0", Rotation="0,0,0", Scale="1,1,1") {
    rotationMode=BMM_Additive
    rotationSpace=BCS_ParentBoneSpace
    scaleSpace=BCS_ParentBoneSpace
    translationMode=BMM_Additive
    translationSpace=BCS_ParentBoneSpace
    alphaBoolBlend {blendOption=Linear}
    boneToModify {boneName=brow_glasses_r_05}
   }
   n26 = Get C_glabella_up_pose() -> (C_glabella_up_pose:float)
   n27 = "Transform (Modify) Bone - Bone: brow_glasses_l_01"(ComponentPose=@n25, Alpha=@n26, Translation="0.026,0.001,0.078", Rotation="0,-0.001,0.021", Scale="1,1,1") {
    rotationMode=BMM_Additive
    rotationSpace=BCS_ParentBoneSpace
    scaleSpace=BCS_ParentBoneSpace
    translationMode=BMM_Additive
    translationSpace=BCS_ParentBoneSpace
    alphaBoolBlend {blendOption=Linear}
    boneToModify {boneName=brow_glasses_l_01}
    rotation {roll=0.021, yaw=-0.001}
    translation=<0.026 0.001 0.078>
   }
   n28 = "Transform (Modify) Bone - Bone: brow_glasses_l_02"(ComponentPose=@n27, Alpha=@n26, Translation="0.000,-0.004,0.322", Rotation="-12.473,-0.012,0.108", Scale="1,1,1") {
    rotationMode=BMM_Additive
    rotationSpace=BCS_ParentBoneSpace
    scaleMode=BMM_Additive
    scaleSpace=BCS_ParentBoneSpace
    translationMode=BMM_Additive
    translationSpace=BCS_ParentBoneSpace
    alphaBoolBlend {blendOption=Linear}
    boneToModify {boneName=brow_glasses_l_02}
    rotation=<rot -12.473 -0.012 0.108>
    translation {y=-0.004, z=0.322}
   }
   n29 = "Transform (Modify) Bone - Bone: brow_glasses_l_03"(ComponentPose=@n28, Alpha=@n26, Translation="-0.196,-0.003,0.470", Rotation="-26.312,-0.090,0.252", Scale="1,1,1") {
    rotationMode=BMM_Additive
    rotationSpace=BCS_ParentBoneSpace
    scaleMode=BMM_Additive
    scaleSpace=BCS_ParentBoneSpace
    translationMode=BMM_Additive
    translationSpace=BCS_ParentBoneSpace
    alphaBoolBlend {blendOption=Linear}
    boneToModify {boneName=brow_glasses_l_03}
    rotation=<rot -26.312 -0.09 0.252>
    translation=<-0.196 -0.003 0.47>
   }
   n30 = "Transform (Modify) Bone - Bone: brow_glasses_l_04"(ComponentPose=@n29, Alpha=@n26, Translation="-0.663,-0.005,0.133", Rotation="-6.404,-0.000,0.050", Scale="1,1,1") {
    rotationMode=BMM_Additive
    rotationSpace=BCS_ParentBoneSpace
    scaleMode=BMM_Additive
    scaleSpace=BCS_ParentBoneSpace
    translationMode=BMM_Additive
    translationSpace=BCS_ParentBoneSpace
    alphaBoolBlend {blendOption=Linear}
    boneToModify {boneName=brow_glasses_l_04}
    rotation=<rot -6.404 0 0.05>
    translation=<-0.663 -0.005 0.133>
   }
   n31 = "Transform (Modify) Bone - Bone: brow_glasses_l_05"(ComponentPose=@n30, Alpha=@n26, Translation="-0.418,-0.002,0.000", Rotation="0,0,0", Scale="1,1,1") {
    rotationMode=BMM_Additive
    rotationSpace=BCS_ParentBoneSpace
    scaleSpace=BCS_ParentBoneSpace
    translationMode=BMM_Additive
    translationSpace=BCS_ParentBoneSpace
    alphaBoolBlend {blendOption=Linear}
    boneToModify {boneName=brow_glasses_l_05}
    translation=<-0.418 -0.002>
   }
   n32 = "Transform (Modify) Bone - Bone: brow_glasses_r_01"(ComponentPose=@n31, Alpha=@n26, Translation="-0.026,0,0", Rotation="0,0,0", Scale="1,1,1") {
    rotationMode=BMM_Additive
    rotationSpace=BCS_ParentBoneSpace
    scaleSpace=BCS_ParentBoneSpace
    translationMode=BMM_Additive
    translationSpace=BCS_ParentBoneSpace
    alphaBoolBlend {blendOption=Linear}
    boneToModify {boneName=brow_glasses_r_01}
    translation {x=-0.026}
   }
   n33 = "Transform (Modify) Bone - Bone: brow_glasses_r_02"(ComponentPose=@n32, Alpha=@n26, Translation="0,0,-0.322", Rotation="-12.473,0,0.108", Scale="1,1,1") {
    rotationMode=BMM_Additive
    rotationSpace=BCS_ParentBoneSpace
    scaleSpace=BCS_ParentBoneSpace
    translationMode=BMM_Additive
    translationSpace=BCS_ParentBoneSpace
    alphaBoolBlend {blendOption=Linear}
    boneToModify {boneName=brow_glasses_r_02}
    rotation {pitch=-12.473, roll=0.108}
    translation {z=-0.322}
   }
   n34 = "Transform (Modify) Bone - Bone: brow_glasses_r_03"(ComponentPose=@n33, Alpha=@n26, Translation="0.196,0,-0.47", Rotation="-26.312,-0.09,0.252", Scale="1,1,1") {
    rotationMode=BMM_Additive
    rotationSpace=BCS_ParentBoneSpace
    scaleMode=BMM_Additive
    scaleSpace=BCS_ParentBoneSpace
    translationMode=BMM_Additive
    translationSpace=BCS_ParentBoneSpace
    alphaBoolBlend {blendOption=Linear}
    boneToModify {boneName=brow_glasses_r_03}
    rotation=<rot -26.312 -0.09 0.252>
    translation {x=0.196, z=-0.47}
   }
   n35 = "Transform (Modify) Bone - Bone: brow_glasses_r_04"(ComponentPose=@n34, Alpha=@n26, Translation="0.663,0,0.133", Rotation="-6.404,0,0", Scale="1,1,1") {
    rotationMode=BMM_Additive
    rotationSpace=BCS_ParentBoneSpace
    scaleSpace=BCS_ParentBoneSpace
    translationMode=BMM_Additive
    translationSpace=BCS_ParentBoneSpace
    alphaBoolBlend {blendOption=Linear}
    boneToModify {boneName=brow_glasses_r_04}
    rotation {pitch=-6.404}
    translation {x=0.663, z=0.133}
   }
   n36 = "Transform (Modify) Bone - Bone: brow_glasses_r_05"(ComponentPose=@n35, Alpha=@n26, Translation="0.418,0,0", Rotation="0,0,0", Scale="1,1,1") {
    rotationMode=BMM_Additive
    rotationSpace=BCS_ParentBoneSpace
    scaleSpace=BCS_ParentBoneSpace
    translationMode=BMM_Additive
    translationSpace=BCS_ParentBoneSpace
    alphaBoolBlend {blendOption=Linear}
    boneToModify {boneName=brow_glasses_r_05}
    translation {x=0.418}
   }
   n37 = Get C_glabella_down_pose() -> (C_glabella_down_pose:float)
   n38 = "Transform (Modify) Bone - Bone: brow_glasses_l_01"(ComponentPose=@n36, Alpha=@n37, Translation="-0.199,-0.003,0.000", Rotation="-13.691,-0.022,0.118", Scale="1,1,1") {
    rotationMode=BMM_Additive
    rotationSpace=BCS_ParentBoneSpace
    scaleSpace=BCS_ParentBoneSpace
    translationMode=BMM_Additive
    translationSpace=BCS_ParentBoneSpace
    alphaBoolBlend {blendOption=Linear}
    boneToModify {boneName=brow_glasses_l_01}
    rotation=<rot -13.691 -0.022 0.118>
    translation=<-0.199 -0.003>
   }
   n39 = "Transform (Modify) Bone - Bone: brow_glasses_l_02"(ComponentPose=@n38, Alpha=@n37, Translation="-0.789,-0.007,-0.000", Rotation="-6.124,-0.004,0.053", Scale="1,1,1") {
    rotationMode=BMM_Additive
    rotationSpace=BCS_ParentBoneSpace
    scaleMode=BMM_Additive
    scaleSpace=BCS_ParentBoneSpace
    translationMode=BMM_Additive
    translationSpace=BCS_ParentBoneSpace
    alphaBoolBlend {blendOption=Linear}
    boneToModify {boneName=brow_glasses_l_02}
    rotation=<rot -6.124 -0.004 0.053>
    translation=<-0.789 -0.007>
   }
   n40 = "Transform (Modify) Bone - Bone: brow_glasses_l_03"(ComponentPose=@n39, Alpha=@n37, Translation="-0.236,0.000,-0.747", Rotation="19.995,-0.027,-0.142", Scale="1,1,1") {
    rotationMode=BMM_Additive
    rotationSpace=BCS_ParentBoneSpace
    scaleMode=BMM_Additive
    scaleSpace=BCS_ParentBoneSpace
    translationMode=BMM_Additive
    translationSpace=BCS_ParentBoneSpace
    alphaBoolBlend {blendOption=Linear}
    boneToModify {boneName=brow_glasses_l_03}
    rotation=<rot 19.995 -0.027 -0.142>
    translation=<-0.2364 0 -0.7474>
   }
   n41 = "Transform (Modify) Bone - Bone: brow_glasses_l_04"(ComponentPose=@n40, Alpha=@n37, Translation="0,0,0", Rotation="0,0,0", Scale="1,1,1") {
    rotationMode=BMM_Additive
    rotationSpace=BCS_ParentBoneSpace
    scaleMode=BMM_Additive
    scaleSpace=BCS_ParentBoneSpace
    translationMode=BMM_Additive
    translationSpace=BCS_ParentBoneSpace
    alphaBoolBlend {blendOption=Linear}
    boneToModify {boneName=brow_glasses_l_04}
   }
   n42 = "Transform (Modify) Bone - Bone: brow_glasses_l_05"(ComponentPose=@n41, Alpha=@n37, Translation="0,0,0", Rotation="0,0,0", Scale="1,1,1") {
    rotationMode=BMM_Additive
    rotationSpace=BCS_ParentBoneSpace
    scaleSpace=BCS_ParentBoneSpace
    translationMode=BMM_Additive
    translationSpace=BCS_ParentBoneSpace
    alphaBoolBlend {blendOption=Linear}
    boneToModify {boneName=brow_glasses_l_05}
   }
   n43 = "Transform (Modify) Bone - Bone: brow_glasses_r_01"(ComponentPose=@n42, Alpha=@n37, Translation="0.199,0,0", Rotation="-13.691,0,0.118", Scale="1,1,1") {
    rotationMode=BMM_Additive
    rotationSpace=BCS_ParentBoneSpace
    scaleSpace=BCS_ParentBoneSpace
    translationMode=BMM_Additive
    translationSpace=BCS_ParentBoneSpace
    alphaBoolBlend {blendOption=Linear}
    boneToModify {boneName=brow_glasses_r_01}
    rotation {pitch=-13.691, roll=0.118}
    translation {x=0.199}
   }
   n44 = "Transform (Modify) Bone - Bone: brow_glasses_r_02"(ComponentPose=@n43, Alpha=@n37, Translation="0.789,0,0", Rotation="-6.124,0,0", Scale="1,1,1") {
    rotationMode=BMM_Additive
    rotationSpace=BCS_ParentBoneSpace
    scaleSpace=BCS_ParentBoneSpace
    translationMode=BMM_Additive
    translationSpace=BCS_ParentBoneSpace
    alphaBoolBlend {blendOption=Linear}
    boneToModify {boneName=brow_glasses_r_02}
    rotation {pitch=-6.124}
    translation {x=0.789}
   }
   n45 = "Transform (Modify) Bone - Bone: brow_glasses_r_03"(ComponentPose=@n44, Alpha=@n37, Translation="0.236,0,0.747", Rotation="19.995,-0.027,-0.142", Scale="1,1,1") {
    rotationMode=BMM_Additive
    rotationSpace=BCS_ParentBoneSpace
    scaleMode=BMM_Additive
    scaleSpace=BCS_ParentBoneSpace
    translationMode=BMM_Additive
    translationSpace=BCS_ParentBoneSpace
    alphaBoolBlend {blendOption=Linear}
    boneToModify {boneName=brow_glasses_r_03}
    rotation=<rot 19.995 -0.027 -0.142>
    translation {x=0.236, z=0.747}
   }
   n46 = "Transform (Modify) Bone - Bone: brow_glasses_r_04"(ComponentPose=@n45, Alpha=@n37, Translation="0,0,0", Rotation="0,0,0", Scale="1,1,1") {
    rotationMode=BMM_Additive
    rotationSpace=BCS_ParentBoneSpace
    scaleSpace=BCS_ParentBoneSpace
    translationMode=BMM_Additive
    translationSpace=BCS_ParentBoneSpace
    alphaBoolBlend {blendOption=Linear}
    boneToModify {boneName=brow_glasses_r_04}
   }
   n47 = "Transform (Modify) Bone - Bone: brow_glasses_r_05"(ComponentPose=@n46, Alpha=@n37, Translation="0,0,0", Rotation="0,0,0", Scale="1,1,1") {
    rotationMode=BMM_Additive
    rotationSpace=BCS_ParentBoneSpace
    scaleSpace=BCS_ParentBoneSpace
    translationMode=BMM_Additive
    translationSpace=BCS_ParentBoneSpace
    alphaBoolBlend {blendOption=Linear}
    boneToModify {boneName=brow_glasses_r_05}
   }
   Output Pose(Result=Component To Local(ComponentPose=@n47)) {layerGroup=None}
  ]}
 EventGraph {nodes [
   Event BlueprintUpdateAnimation()
   > Is Valid(InputObject=Get HeadPartAnimInstance() -> (HeadPartAnimInstance:Anim Instance Object Reference))
   > n1 = Sequence!(execute) -> (then_0, then_1, then_2:Exec, then_3:Exec, then_4:Exec, then_5:Exec, then_6:Exec, then_7:Exec, then_8:Exec, then_9:Exec)
   > Set l_brow_up_pose(l_brow_up_pose:float=GetCurveValue(self=Get HeadPartAnimInstance(), CurveName=L_brow_up_pose)) -> (Output_Get:float)
   Set disable_zebrascramble_fingers(execute=@n1.then_8, disable_zebrascramble_fingers:float=GetCurveValue(CurveName=disable_zebrascramble_fingers)) -> (Output_Get:float)
   n2 = GetCurveValue(self=Get HeadPartAnimInstance(), CurveName=R_brow_down_pose)
   n3 = Branch(execute=@n1.then_3, Condition="float > float"(A=@n2, B=0.7))
   > Set R_brow_down_pose(R_brow_down_pose:float=1) -> (Output_Get:float)
   Set R_brow_down_pose(execute=@n3.else, R_brow_down_pose=0)
   Set R_brow_up_pose(execute=@n1.then_2, R_brow_up_pose:float=GetCurveValue(self=Get HeadPartAnimInstance(), CurveName=R_brow_up_pose)) -> (Output_Get:float)
   n4 = GetCurveValue(self=Get HeadPartAnimInstance(), CurveName=L_brow_down_pose)
   n5 = Branch(execute=@n1.then_1, Condition="float > float"(A=@n4, B=0.7))
   > Set L_brow_down_pose(L_brow_down_pose:float=1) -> (Output_Get:float)
   Set L_brow_down_pose(execute=@n5.else, L_brow_down_pose=0)
   Set L_brow_squeeze_in_pose(execute=@n1.then_4, L_brow_squeeze_in_pose:float=GetCurveValue(self=Get HeadPartAnimInstance(), CurveName=L_brow_squeeze_in_pose)) -> (Output_Get:float)
   Set R_brow_squeeze_in_pose(execute=@n1.then_5, R_brow_squeeze_in_pose:float=GetCurveValue(self=Get HeadPartAnimInstance(), CurveName=R_brow_squeeze_in_pose)) -> (Output_Get:float)
   n6 = Get L_brow_down_pose()
   n7 = Get R_brow_down_pose()
   n8 = "float < float"(A=float * float(A=float + float(A=@n6, B=@n7), B=0.5) {is=title K2Node_PromotableOperator}, B=0.5)
   n9 = Branch(execute=@n1.then_7, Condition=@n8)
   n10 = Get HeadPartAnimInstance()
   > Set C_glabella_down_pose(C_glabella_down_pose:float=MapRangeClamped(Value=GetCurveValue(self=@n10, CurveName=C_glabella_down_pose), InRangeA=1, InRangeB=2)) -> (Output_Get:float)
   Set C_glabella_down_pose(execute=@n9.else, C_glabella_down_pose=0)
   n11 = Branch(execute=@n1.then_6, Condition=@n8)
   n12 = Get HeadPartAnimInstance()
   > Set C_glabella_up_pose(C_glabella_up_pose:float=MapRangeClamped(Value=GetCurveValue(self=@n12, CurveName=C_glabella_up_pose), InRangeA=0.5, InRangeB=1)) -> (Output_Get:float)
   Set C_glabella_up_pose(execute=@n11.else, C_glabella_up_pose=0)
   Event BlueprintPostEvaluateAnimation() {is=title K2Node_Event}
   > Delay(Duration=0)
   > UpdateEyeVirtualPlane()
   n13 = Event OnCharacterCustomizationLoaded()
   > Set HeadPartAnimInstance(HeadPartAnimInstance:Anim Instance Object Reference=GetAnimInstance(self=Get Array Item with EFort Custom Part Type(Array=@n13.CustomMeshParts, PartType=Head))) -> (Output_Get:Anim Instance Object Reference)
   MapRangeClamped(Value=@n4, InRangeA=0.5, InRangeB=1)
   MapRangeClamped(Value=@n2, InRangeA=0.5, InRangeB=1)
  ]}
 UpdateEyeVirtualPlane {nodes [
   UpdateEyeVirtualPlane() {is=title K2Node_FunctionEntry}
   > n1 = Get Dynamic Material Instance(SkeletalMeshComponent=GetPartSkeletalMeshComponent(), ElementIndex=2)
   n2 = BreakTransform(InTransform=GetBoneTransform!(
     self:Skinned Mesh Component Object Reference=GetPartSkeletalMeshComponent()
     InBoneName:Name=brow_glasses_r_root
     TransformSpace:ERelativeTransformSpace Enum=RTS_World
    ) -> (ReturnValue:Transform))
   n3 = GetRightVector!(InRot:Rotator=@n2.Rotation) -> (ReturnValue:Vector)
   > SetVectorParameterValue(
    self=@n1
    ParameterName=VP_NormalAxis
    Value="To LinearColor (Vector)"!(InVec:Vector=vector * vector(A=@n3, B:float=1) -> (ReturnValue=~)) -> (ReturnValue:Linear Color Structure)
   )
   n4 = GetUpVector!(InRot:Rotator=@n2.Rotation) -> (ReturnValue:Vector)
   > SetVectorParameterValue(self=@n1, ParameterName=VP_XAxis, Value="To LinearColor (Vector)"(InVec=vector * vector(A=@n4, B=-1) -> (ReturnValue=~)))
   n5 = Get!() -> (RightEyeOffset_X:double, RightEyeOffset_Y:double, RightEyeOffset_Z:double)
   n6 = vector * vector(A=@n3, B:double=@n5) -> (ReturnValue=~)
   n7 = vector * vector(A=@n4, B=@n5.RightEyeOffset_Y) -> (ReturnValue=~)
   n8 = GetForwardVector!(InRot:Rotator=@n2.Rotation) -> (ReturnValue:Vector)
   n9 = vector * vector(A=@n8, B=@n5.RightEyeOffset_Z) -> (ReturnValue=~)
   > SetVectorParameterValue(
    self=@n1
    ParameterName=VP_RightEyePos
    Value="To LinearColor (Vector)"(InVec=vector + vector!(A=@n2, B=@n6, C:Vector=@n7, D:Vector=@n9) -> (ReturnValue))
   )
   n10 = BreakTransform(InTransform=GetBoneTransform(self=GetPartSkeletalMeshComponent(), InBoneName=brow_glasses_l_root, TransformSpace=RTS_World))
   n11 = Get!() -> (LeftEyeOffset_X:double, LeftEyeOffset_Y:double, LeftEyeOffset_Z:double)
   n12 = vector * vector(A=@n4, B=@n11.LeftEyeOffset_Y) -> (ReturnValue=~)
   n13 = vector * vector(A=@n3, B=@n11) -> (ReturnValue=~)
   n14 = vector * vector(A=@n8, B=@n11.LeftEyeOffset_Z) -> (ReturnValue=~)
   > SetVectorParameterValue(self=@n1, ParameterName=VP_LeftEyePos, Value="To LinearColor (Vector)"(InVec=vector + vector(A=@n10, B=@n13, C=@n12, D=@n14)))
   > SetScalarParameterValue(self=@n1, ParameterName=VP_UseLocalSpace, Value=0)
  ]}
]
variables [
 C_glabella_down_pose {type=number}
 C_glabella_up_pose {type=number}
 disable_zebrascramble_fingers {type=number}
 HeadPartAnimInstance {type=AnimInstance, default=null}
 Is Bacon {type=boolean}
 L_brow_down_pose {type=number}
 L_brow_squeeze_in_pose {type=number}
 l_brow_up_pose {type=number}
 R_brow_down_pose {type=number}
 R_brow_squeeze_in_pose {type=number}
 R_brow_up_pose {type=number}
]
